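# Sound cue for the player heal loop in the Uranium LTM, reusing a Slurp barrel heal sound.
# A Mixer root boosts a single wave player and has two empty inputs; the wave player does not enable looping, despite the cue's name.
# It uses consumable sound class and attenuation, with reduced cue volume.
# A concurrency override caps simultaneous instances at four, stopping the oldest.
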SoundCue /FNGameplayCues/Sounds/LTM/Uranium/Uranium_Player_Heal_Loop_Cue
root=Mixer_0
nodes [
 {id=Mixer_0, children [WavePlayer_0,null,null], properties {inputVolume [1.5,2,1]}}
 {id=WavePlayer_0, properties {soundWaveAssetPtr=/FNGameplayCues/Athena/Sounds/Items/Slurp/Item_Slurp_Barrel_Heal_04}}
]
properties {
 attenuationSettings=&Fort_Consumable_Use_Athena_Att
 bOverrideConcurrency=true
 maxDistance=2250
 soundClassObject=&Fort_Consumables
 volumeMultiplier=0.5
 concurrencyOverrides {maxCount=4, resolutionRule=StopOldest}
}
